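# One-shot 5.1 MetaSound Source for a front-end UI cinematic whoosh, named for a video transition out.
# It wraps the shared MSS_PlayRandomSound_5_1 source, passing in the Sounds array and a narrow random pitch range at full volume.
# All six channels and the finished trigger route straight to the outputs.
# The default Sounds entry is null; playback uses the RenderedCinematics sound class with virtualization disabled.
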MetaSoundSource /Game/Sounds/Fort_UI_Sounds/FrontEnd/Cinematics_UI/MSS_UI_Cine_Video_TransitionOut_Whoosh
inputs [
 MaxPitch {type=Float, access=Constructor}
 MinPitch {type=Float, access=Constructor}
 Sounds {type=WaveAsset:Array, access=Constructor}
 UE.Source.OnPlay {type=Trigger}
]
interfaces [UE.OutputFormat.5dot1:1.0, UE.Source.OneShot:1.0, UE.Source:1.1]
outputs [
 "UE.OutputFormat.5dot1.Audio:0" {type=Audio}
 "UE.OutputFormat.5dot1.Audio:1" {type=Audio}
 "UE.OutputFormat.5dot1.Audio:2" {type=Audio}
 "UE.OutputFormat.5dot1.Audio:3" {type=Audio}
 "UE.OutputFormat.5dot1.Audio:4" {type=Audio}
 "UE.OutputFormat.5dot1.Audio:5" {type=Audio}
 UE.Source.OneShot.OnFinished {type=Trigger}
]
pages [Default {
  input_defaults {MaxPitch=1.05, MinPitch=0.95, UE.Source.OnPlay=null, Sounds [null]}
  nodes [MSS_PlayRandomSound_5_1_0 = /FortMetasound/Sources/MSS_PlayRandomSound_5_1(MaxPitch=^MaxPitch, MinPitch=^MinPitch, Sounds=^Sounds, UE.Source.OnPlay=^UE.Source.OnPlay) {also {defaults {MaxVolume=1, MinVolume=1}}}]
  links [
   {dst="@UE.OutputFormat.5dot1.Audio:0", src=MSS_PlayRandomSound_5_1_0.UE.OutputFormat.5dot1.Audio:0}
   {dst="@UE.OutputFormat.5dot1.Audio:1", src=MSS_PlayRandomSound_5_1_0.UE.OutputFormat.5dot1.Audio:1}
   {dst="@UE.OutputFormat.5dot1.Audio:2", src=MSS_PlayRandomSound_5_1_0.UE.OutputFormat.5dot1.Audio:2}
   {dst="@UE.OutputFormat.5dot1.Audio:3", src=MSS_PlayRandomSound_5_1_0.UE.OutputFormat.5dot1.Audio:3}
   {dst="@UE.OutputFormat.5dot1.Audio:4", src=MSS_PlayRandomSound_5_1_0.UE.OutputFormat.5dot1.Audio:4}
   {dst="@UE.OutputFormat.5dot1.Audio:5", src=MSS_PlayRandomSound_5_1_0.UE.OutputFormat.5dot1.Audio:5}
   {dst="@UE.Source.OneShot.OnFinished", src=MSS_PlayRandomSound_5_1_0.UE.Source.OneShot.OnFinished}
  ]
 }]
properties {
 assetImportData=&self:AssetImportData
 numChannels=6
 outputFormat=FiveDotOne
 sampleRate=48000
 soundClassObject=/Game/Sounds/FortSoundClassesAndModes/RenderedCinematics
 virtualizationMode=Disabled
}
